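# Translucent UI material that draws a tinted radial glow, named as the center glow of the MistRaven HUD reticle.
# The ValidTarget parameter lerps a scale-from-center on the UVs, resizing the glow between nearly collapsed and slightly enlarged.
# A radial gradient function's smooth output becomes the glow opacity and, multiplied by GlowColor, the glow color.
# Both results end in named reroute declarations, with no material output connections shown.
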Material /MistRaven/HUD/CenterReticle/M_UI_MistRaven_Reticle_CenterGlow
graph {
 properties {bAutomaticallySetUsageInEditor=false, blendMode=BLEND_Translucent, materialDomain=MD_UI, usedShadingModels=Default Lit}
 nodes [
  n1 = ScalarParameter() {defaultValue=1, parameterName=ValidTarget}
  n2 = Constant2Vector() {g=1}
  n3 = MaterialFunctionCall(
   UV=MaterialFunctionCall(Texture Scale=LinearInterpolate(Alpha=@n1) {constA=0.01, constB=1.2}) {materialFunction=&ScaleUVsByCenter}.UVs
   GradientStart=@n2.R
   GradientEnd=@n2.R
  ) {materialFunction=&MF_UI_RadialGradient}
  NamedRerouteDeclaration(None=@n3.Smooth) {name=BackingGlowOpacity, nodeColor=<rgba 0.3828 0 1 1>}
  NamedRerouteDeclaration(None=Multiply(A=VectorParameter() {parameterName=GlowColor, defaultValue=<rgba 1 0.0048 0.0595 1>}.RGB, B=@n3.Smooth)) {name=BackingGlowColor, nodeColor=<rgba 0 1 0.0625 1>}
 ]
}
