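# Translucent UI material for the MistRaven center reticle, drawing a circle-cross shape over a radial backing glow.
# The cross is a procedural distance field from mirrored, centered UVs and circle distances, anti-aliased into a tinted fill and softer shadow.
# The ValidTarget parameter drives the glow's scale and the shadow's blend.
# Named reroutes carry the layers into a color blend and a max-combined alpha scaled by ReticleOpacity.
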Material /MistRaven/HUD/CenterReticle/M_UI_MistRaven_Reticle_CircleCross
graph {
 properties {bAutomaticallySetUsageInEditor=false, blendMode=BLEND_Translucent, materialDomain=MD_UI, usedShadingModels=Default Lit}
 nodes [
  n1 = NamedRerouteUsage()
  LinearInterpolate(A=NamedRerouteUsage(), B=NamedRerouteUsage(), Alpha=@n1)
  Multiply(A=Max(A=@n1, B=NamedRerouteUsage()), B=ScalarParameter() {defaultValue=0.85, parameterName=ReticleOpacity})
  n2 = ScalarParameter() {parameterName=ValidTarget}
  n3 = Constant2Vector() {g=1}
  n4 = MaterialFunctionCall(
   UV=MaterialFunctionCall(Texture Scale=LinearInterpolate(Alpha=@n2) {constA=0.01, constB=1.2}) {materialFunction=&ScaleUVsByCenter}.UVs
   GradientStart=@n3.R
   GradientEnd=@n3.R
  ) {materialFunction=&MF_UI_RadialGradient}
  NamedRerouteDeclaration(None=@n4.Smooth) {name=BackingGlowOpacity, nodeColor=<rgba 0.3828 0 1 1>}
  NamedRerouteDeclaration(None=Multiply(A=VectorParameter() {parameterName=GlowColor, defaultValue=<rgba 1 0 0 1>}.RGB, B=@n4.Smooth)) {name=BackingGlowColor, nodeColor=<rgba 0 1 0.0625 1>}
  n5 = NamedRerouteUsage()
  n6 = Subtract(A=TextureCoordinate()) {constB=0.5}
  n7 = Divide(B=ScalarParameter() {defaultValue=0.6, group=CircleCross, parameterName=CircleCrossScale}) {constA=2}
  MaterialExpressionNamedRerouteDeclaration_41 = NamedRerouteDeclaration(None=Abs(None=Multiply(A=@n6, B=@n7))) {name=CircleCrossUVs, nodeColor=<rgba 0 0 0 1>}
  n8 = Convert(None=@MaterialExpressionNamedRerouteDeclaration_41) {
   convertInputs [{
     type=Vector2
     expressionInput {
      expression=&self:MaterialExpressionNamedRerouteDeclaration_41
      inputName=None
      mask=0
      maskA=0
      maskB=0
      maskG=0
      maskR=0
      outputIndex=0
     }
    }]
   convertOutputs [{type=Scalar}, {type=Scalar}]
  }
  n9 = Multiply(A=@n8, B=Subtract(A=NamedRerouteUsage()))
  n10 = ScalarParameter() {defaultValue=1, group=CircleCross, parameterName=CircleCrossHeight}
  n11 = MaterialFunctionCall!(
   A=MaterialFunctionCall(A=@n8) {materialFunction=/FortUILibrary/MaterialFunctions/Basics/MF_UI_BooleanOperations}.IsLess
   B=MaterialFunctionCall(A=@n8, B=Add(A=@n9, B=@n10)) {materialFunction=&MF_UI_BooleanOperations}.IsLess
  ) {materialFunction=&MF_UI_BooleanOperations}
  n12 = NamedRerouteUsage()
  n13 = Subtract(
   A=Max(A=Multiply(A=@n5, B=@n11.And), B=Multiply(A=@n11.Not And, B=@n12))
   B=ScalarParameter() {defaultValue=0.2, group=CircleCross, parameterName=CircleCrossPointRoundness}
  )
  n14 = MaterialFunctionCall(DistanceField=@n13) {materialFunction=&MF_UI_AntiAliasing}
  n15 = Constant2Vector() {r=0.15}
  n16 = ScalarParameter() {parameterName=ValidTarget}
  NamedRerouteDeclaration(None=LinearInterpolate(
    A=@n14.Inner
    B=MaterialFunctionCall(DistanceField=@n13, Threshold=@n15.R) {materialFunction=&MF_UI_AntiAliasing}.Inner
    Alpha=LinearInterpolate(Alpha=@n16) {constA=0.5, constB=0.8}
   ) {constAlpha=0.8}) {name=CircleCrossShadow, nodeColor=<rgba 0 0 0 1>}
  n17 = ScalarParameter() {defaultValue=1, group=CircleCross, parameterName=CircleCrossHeight}
  n18 = Add(A=@n17)
  NamedRerouteDeclaration(None=Multiply(A=Divide(A=@n18, B=@n17) {constB=0.5}) {constB=0.5}) {name=CircleCrossHeight, nodeColor=<rgba 0 0 0 1>}
  n19 = Subtract(A=NamedRerouteUsage(), B=AppendVector(A=Constant() {r=1}, B=NamedRerouteUsage()))
  n20 = DotProduct(A=@n19, B=@n19)
  NamedRerouteDeclaration(None=Subtract(A=NamedRerouteUsage(), B=SquareRoot(None=@n20))) {name=CircleCrossTrue, nodeColor=<rgba 0 0 0 1>}
  n21 = Subtract(
   A=NamedRerouteUsage()
   B=AppendVector(A=Constant(), B=ScalarParameter() {defaultValue=1, group=CircleCross, parameterName=CircleCrossHeight})
  )
  n22 = DotProduct(A=@n21, B=@n21)
  n23 = Subtract(A=NamedRerouteUsage(), B=Constant2Vector() {r=1})
  n24 = DotProduct(A=@n23, B=@n23)
  NamedRerouteDeclaration(None=SquareRoot(None=Min(A=@n22, B=@n24))) {name=CircleCrossFalse, nodeColor=<rgba 0 0 0 1>}
  NamedRerouteDeclaration(None=Multiply(A=VectorParameter() {parameterName=CircleCrossColor, defaultValue=<rgba 1 1 1 0>}.RGB, B=@n14.Inner)) {name=CircleCrossColor, nodeColor=<rgba 0 0 0 1>}
 ]
}
