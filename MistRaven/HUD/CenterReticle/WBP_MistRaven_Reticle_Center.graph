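# MistRaven center reticle extension widget: a circle-cross over a glow that signals whether a valid target exists.
# On construct it tints the glow by the pawn's all-ascended tag and starts async gameplay-tag listeners on the player controller.
# Tendril vacuum and wraith form tags toggle widget visibility, while the no-valid-target tag plays the glow material's ValidTarget animation forward or reverse.
# Attack and cooldown listeners are defined but never called here.
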WidgetBlueprint /MistRaven/HUD/CenterReticle/WBP_MistRaven_Reticle_Center
animations [Anim_TargetValid {
  binding_count=1
  length_seconds=0.15
  object=&self:Anim_TargetValid
  sequence {bindings [Image_Glow {
     id=#1
     possessed_class=Image
     tracks [{
       type=MovieSceneWidgetMaterialTrack
       sections [{
         type=MovieSceneParameterSection
         end_frame=3
         start_frame=0
         channels [ValidTarget {type=MovieSceneScriptingFloatChannel, keys [[0,0],[3,1]]}]
        }]
      }]
    }]}
 }]
graphs [
 EventGraph {nodes [
   ListenForWraithFormActive() {is=title K2Node_CustomEvent}
   > n1 = WaitGameplayTagAddToActor!(
    execute:Exec
    TargetActor:Actor Object Reference=Get PlayerController() -> (PlayerController:Player Controller Object Reference)
    Tag:Gameplay Tag Structure=Get WraithFormActiveTag() -> (WraithFormActiveTag:Gameplay Tag Structure)
    OnlyTriggerOnce:bool=false
   ) -> (then:Exec, AsyncTaskProxy:Ability Async Wait Gameplay Tag Added Object Reference, Added:Exec) {is=title K2Node_AsyncAction}
   > n2 = WaitGameplayTagRemoveFromActor!(
    execute:Exec
    TargetActor:Actor Object Reference=Get PlayerController()
    Tag:Gameplay Tag Structure=Get WraithFormActiveTag()
    OnlyTriggerOnce:bool=false
   ) -> (then:Exec, AsyncTaskProxy:Ability Async Wait Gameplay Tag Removed Object Reference, Removed:Exec) {is=title K2Node_AsyncAction}
   HandleWraithFormActive!(execute=@n2.Removed, self, IsWraithFormActive:bool=false) -> (then)
   HandleWraithFormActive(execute=@n1.Added, IsWraithFormActive=true)
   ListenForTendrilVacuumEquip() {is=title K2Node_CustomEvent}
   > n3 = WaitGameplayTagAddToActor(
    TargetActor=Get PlayerController()
    Tag=Get TendrilVaccumGameplayTag() -> (TendrilVaccumGameplayTag:Gameplay Tag Structure)
    OnlyTriggerOnce=false
   )
   > n4 = WaitGameplayTagRemoveFromActor(TargetActor=Get PlayerController(), Tag=Get TendrilVaccumGameplayTag(), OnlyTriggerOnce=false)
   HandleTendrilVacuumEquip!(execute=@n4.Removed, self, IsEquipped:bool=false) -> (then)
   HandleTendrilVacuumEquip(execute=@n3.Added, IsEquipped=true)
   OnEquip() {is=title K2Node_CustomEvent}
   > ListenForTendrilVacuumEquip() {is=title K2Node_CallFunction}
   > ListenForWraithFormActive() {is=title K2Node_CallFunction}
   > ListenForValidTarget()
   ListenForWraithFormCooldown() {is=title K2Node_CustomEvent}
   > n5 = WaitGameplayTagAddToActor(TargetActor=Get PlayerController(), Tag=Get WraithFormCooldownTag() -> (WraithFormCooldownTag:Gameplay Tag Structure), OnlyTriggerOnce=false)
   > n6 = WaitGameplayTagRemoveFromActor(TargetActor=Get PlayerController(), Tag=Get WraithFormCooldownTag(), OnlyTriggerOnce=false)
   HandleWraithFormCooldown!(execute=@n6.Removed, self, IsCooldownActive:bool=false) -> (then)
   HandleWraithFormCooldown(execute=@n5.Added, IsCooldownActive=true)
   ListenForValidTarget() {is=title K2Node_CustomEvent}
   > n7 = WaitGameplayTagAddToActor(
    TargetActor=Get PlayerController()
    Tag=Get NoValidTargetGameplayTag() -> (NoValidTargetGameplayTag:Gameplay Tag Structure)
    OnlyTriggerOnce=false
   )
   > n8 = WaitGameplayTagRemoveFromActor(TargetActor=Get PlayerController(), Tag=Get NoValidTargetGameplayTag(), OnlyTriggerOnce=false)
   HandleValidTarget!(execute=@n8.Removed, self, IsTargetValid:bool=true) -> (then)
   HandleValidTarget(execute=@n7.Added, IsTargetValid=false)
   ListenForAttack() {is=title K2Node_CustomEvent}
   > n9 = WaitGameplayTagAddToActor(TargetActor=Get PlayerController(), Tag=Get AttackGameplayTag() -> (AttackGameplayTag:Gameplay Tag Structure), OnlyTriggerOnce=false)
   > n10 = WaitGameplayTagRemoveFromActor(TargetActor=Get PlayerController(), Tag=Get AttackGameplayTag(), OnlyTriggerOnce=false)
   HandleAttack(execute=@n10.Removed)
   HandleAttack(execute=@n9.Added)
   Event Construct()
   > Parent: Construct()
   > Set PlayerController(PlayerController:Player Controller Object Reference=GetOwningPlayer()) -> (Output_Get:Player Controller Object Reference)
   > Set PlayerPawn(PlayerPawn:Pawn Object Reference=GetOwningPlayerPawn!(self:User Widget Object Reference) -> (ReturnValue:Pawn Object Reference)) -> (Output_Get:Pawn Object Reference)
   > OnEquip() {is=title K2Node_CallFunction}
   > n11 = GetDynamicMaterial(self=Get Image_Glow() -> (Image_Glow:Image Object Reference))
   > n12 = Set GlowMaterialRef(GlowMaterialRef:Material Instance Dynamic Object Reference=@n11) -> (Output_Get:Material Instance Dynamic Object Reference)
   > SetScalarParameterValue(self=@n12, ParameterName=ValidTarget, Value=0)
   > n13 = Cast To B_MistRaven_FlameThrowerAscended!(
    execute
    Object=GetEquippedWeapon!(self:Fort Weapon Reticle Extension Widget Base Object Reference) -> (ReturnValue:Fort Weapon Object Reference)
   ) -> (then, CastFailed, AsB Mist Raven Flame Thrower Ascended:B Mist Raven Flame Thrower Ascended Object Reference) {is=title K2Node_DynamicCast}
   > Set AsBMistRavenFlameThrower(AsBMistRavenFlameThrower:B Mist Raven Flame Thrower Ascended Object Reference=@n13) -> (Output_Get:B Mist Raven Flame Thrower Ascended Object Reference)
   > n14 = Cast To FortPlayerPawn(Object=Get PlayerPawn() -> (PlayerPawn:Pawn Object Reference))
   n15 = Get AscendedEquippedTag() -> (AscendedEquippedTag:Gameplay Tag Structure)
   > SetVectorParameterValue(
    self=Get GlowMaterialRef() -> (GlowMaterialRef:Material Instance Dynamic Object Reference)
    ParameterName=GlowColor
    Value=SelectColor(
     A=Get GlowCombinedColor() -> (GlowCombinedColor:Linear Color Structure)
     B=Get GlowNormalColor() -> (GlowNormalColor:Linear Color Structure)
     bPickA=HasMatchingGameplayTag!(self:Gameplay Tag Asset Interface Interface=@n14, TagToCheck:Gameplay Tag Structure=@n15) -> (ReturnValue:bool)
    )
   )
  ]}
 HandleAttack {nodes [
   HandleAttack() {is=title K2Node_FunctionEntry}
   > n1 = Branch(Condition=Get IsTargetValid() -> (IsTargetValid:bool))
   > PlayAnimationForward(InAnimation=Get Anim_TargetValid() -> (Anim_TargetValid:Widget Animation Object Reference))
   PlayAnimationReverse(execute=@n1.else, InAnimation=Get Anim_TargetValid(), PlaybackSpeed=1.5)
  ]}
 HandleTendrilVacuumEquip {nodes [
   n1 = HandleTendrilVacuumEquip!() -> (then, IsEquipped:bool=~) {is=title K2Node_FunctionEntry}
   > n2 = Set IsTendrilVacuumEquipped(IsTendrilVacuumEquipped:bool=@n1) -> (Output_Get:bool)
   > n3 = Branch(Condition=@n2)
   > SetVisibility(InVisibility=HitTestInvisible)
   SetVisibility(execute=@n3.else, InVisibility=Collapsed)
  ]}
 HandleValidTarget {nodes [
   n1 = HandleValidTarget!() -> (then, IsTargetValid:bool=~) {is=title K2Node_FunctionEntry}
   > n2 = Set IsTargetValid(IsTargetValid:bool=@n1) -> (Output_Get:bool)
   > n3 = Branch(Condition=@n2)
   > PlayAnimationForward(InAnimation=Get Anim_TargetValid())
   PlayAnimationReverse(execute=@n3.else, InAnimation=Get Anim_TargetValid(), PlaybackSpeed=1.5)
  ]}
 HandleWraithFormActive {nodes [
   n1 = HandleWraithFormActive!() -> (then, IsWraithFormActive:bool=~) {is=title K2Node_FunctionEntry}
   > n2 = Set IsWraithFormActive(IsWraithFormActive:bool=@n1) -> (Output_Get:bool)
   n3 = Get IsTendrilVacuumEquipped() -> (IsTendrilVacuumEquipped:bool)
   n4 = Get IsAscended() -> (IsAscended:bool)
   > n5 = Branch(Condition=OR Boolean(A=@n2, B=OR Boolean(A=@n3, B=@n4)))
   > SetVisibility(InVisibility=HitTestInvisible)
   SetVisibility(execute=@n5.else, InVisibility=Collapsed)
  ]}
 HandleWraithFormCooldown {nodes [
   n1 = HandleWraithFormCooldown!() -> (then, IsCooldownActive:bool=~) {is=title K2Node_FunctionEntry}
   > n2 = Set IsCooldownActive(IsCooldownActive:bool=@n1) -> (Output_Get:bool)
   > n3 = Branch(Condition=@n2)
   n4 = Get IsTendrilVacuumEquipped()
   n5 = Get IsAscended()
   > n6 = Branch(Condition=OR Boolean(A=Get IsWraithFormActive() -> (IsWraithFormActive:bool), B=OR Boolean(A=@n4, B=@n5)))
   > SetVisibility(InVisibility=HitTestInvisible)
   SetVisibility(execute=@n6.else, InVisibility=Collapsed)
   SetVisibility(execute=@n3.else, InVisibility=HitTestInvisible)
  ]}
]
variables [
 AllAscendedEquippedTag {type=GameplayTag, default {tagName=MistRaven.AscendedAcheived}}
 AsBMistRavenFlameThrower {type=B_MistRaven_FlameThrowerAscended_C, default=null}
 AsBMistRavenWeaponAscended {type=B_MistRaven_WeaponAscended_C, default=null}
 AscendedEquippedTag {type=GameplayTag, default {tagName=MistRaven.AllAscendedEquipped}}
 AttackGameplayTag {type=GameplayTag, default {tagName=GameplayCue.Athena.MistRaven.Flamethrower.Active}}
 GlowCombinedColor {type=LinearColor, default=<rgba 0.5029 0 1 1>}
 GlowMaterialRef {type=MaterialInstanceDynamic}
 GlowNormalColor {type=LinearColor, default=<rgba 1 0 0 1>}
 IsAscended {type=boolean}
 IsCooldownActive {type=boolean}
 IsTargetValid {type=boolean}
 IsTendrilVacuumEquipped {type=boolean}
 IsWraithFormActive {type=boolean}
 NoValidTargetGameplayTag {type=GameplayTag, default {tagName=MistRaven.TendrilDrain.Ascended.NoValidTarget}}
 PlayerController {type=PlayerController, default=null}
 PlayerPawn {type=Pawn, default=null}
 TendrilVaccumGameplayTag {type=GameplayTag, default {tagName=MistRaven.TendrilDrain.Ascended.InEquipemtn}}
 WraithFormActiveTag {type=GameplayTag, default {tagName=MistRaven.GrimStrut.Active}}
 WraithFormCooldownTag {type=GameplayTag, default {tagName=Granted.Cooldown.MistRaven.GrimStrut.Ascended.Cooldown}}
]
widgets {
 info {
  inheritedWidgetCount=1
  parentClass=/Script/FortniteUI.FortWeaponReticleExtensionWidgetBase
  rootWidgetClass=&Overlay
  widgetCount=4
 }
 widgets [
  Overlay_1 {widgetClassPath=&Overlay, widget=Overlay_1, properties {visibility=HitTestInvisible}}
  Image_Glow {
   widgetClassPath=&Image
   bIsVariable=true
   parent=Overlay_1
   widget=Image_Glow
   properties {
    renderOpacity=0.7
    visibility=HitTestInvisible
    brush {
     bIsDynamicallyLoaded=false
     drawAs=Image
     imageType=FullColor
     mirroring=NoMirror
     resourceName=None
     resourceObject=&/M_UI_MistRaven_Reticle_CenterGlow
     tiling=NoTile
     imageSize=<24 24>
     margin=<ltrb 0 0 0 0>
     outlineSettings {
      bUseBrushTransparency=false
      roundingType=HalfHeightRadius
      width=0
      color {colorUseRule=UseColor_Specified, specifiedColor=<rgba 0 0 0 0>}
      cornerRadii=<0 0 0 0>
     }
     tintColor {colorUseRule=UseColor_Specified, specifiedColor=<rgba 1 1 1 1>}
     uVRegion {bIsValid=false, max=<0 0>, min=<0 0>}
    }
   }
   slot {class=OverlaySlot, properties {horizontalAlignment=HAlign_Center, verticalAlignment=VAlign_Center}}
  }
  Image_Reticle {
   widgetClassPath=&Image
   bIsVariable=true
   parent=Overlay_1
   widget=Image_Reticle
   properties {
    visibility=HitTestInvisible
    brush {
     bIsDynamicallyLoaded=false
     drawAs=Image
     imageType=FullColor
     mirroring=NoMirror
     resourceName=None
     resourceObject=&/M_UI_MistRaven_Reticle_CircleCross
     tiling=NoTile
     imageSize=<10 10>
     margin=<ltrb 0 0 0 0>
     outlineSettings {
      bUseBrushTransparency=false
      roundingType=HalfHeightRadius
      width=0
      color {colorUseRule=UseColor_Specified, specifiedColor=<rgba 0 0 0 0>}
      cornerRadii=<0 0 0 0>
     }
     tintColor {colorUseRule=UseColor_Specified, specifiedColor=<rgba 1 1 1 1>}
     uVRegion {bIsValid=false, max=<0 0>, min=<0 0>}
    }
   }
   slot {class=OverlaySlot, properties {horizontalAlignment=HAlign_Center, verticalAlignment=VAlign_Center}}
  }
  TutorialHighlight {widgetClassPath=&TutorialHighlight}
 ]
}
